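; Art tool Blueprint that builds a cubemap capture from six high-resolution screenshots.
; Spawn Camera creates a square-aspect camera with cropped overscan at the actor's transform.
; Capture Faces rotates it through six axis directions, running highresshot at each in a chain of one-second timed events.
; Capture Cubemap applies the Face1–6 texture variables to planes via dynamic M_Cubeface materials, then triggers a SceneCaptureComponentCube, toggling Plane6 visibility around the capture.
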(event Custom|Spawn%20Camera
  (bind _camera (Variables|Default|GetCamera))
  (bind _cameracomponent (Variables|/Script/Engine.CameraActor|GetCameraComponent _camera))
  (bind _returnvalue (Game|SpawnActorfromClass "/Script/Engine.CameraActor" (Transformation|GetActorTransform) "Undefined" "MultiplyWithRoot" 0 0))
  (Variables|Default|SetCamera _returnvalue)
  (Camera|SetAspectRatio _cameracomponent 1.0)
  (Camera|SetOverscan _cameracomponent 0.1)
  (Camera|SetCropOverscan _cameracomponent true))

(event Custom|Capture%20Faces
  (CallFunction|HideFaces)
  (Transformation|SetActorRotation (Variables|Default|GetCamera) "0, 0, 0" false)
  (Development|ExecuteConsoleCommand "highresshot 3")
  (Development|PrintString "Face1")
  (Utilities|Time|SetTimerbyEvent (delegate Custom|CaptureFace2) 1.0))

(event Custom|CaptureFace2
  (Transformation|SetActorRotation (Variables|Default|GetCamera) "0.000000,90.000000,0.000000" false)
  (Development|ExecuteConsoleCommand "highresshot 3")
  (Development|PrintString "Face2")
  (Utilities|Time|SetTimerbyEvent (delegate Custom|CaptureFace3) 1.0))

(event Custom|CaptureFace3
  (Transformation|SetActorRotation (Variables|Default|GetCamera) "0.000000,180.000000,0.000000" false)
  (Development|ExecuteConsoleCommand "highresshot 3")
  (Development|PrintString "Face3")
  (Utilities|Time|SetTimerbyEvent (delegate Custom|CaptureFace4) 1.0))

(event Custom|CaptureFace4
  (Transformation|SetActorRotation (Variables|Default|GetCamera) "0.000000,-90.000000,0.000000" false)
  (Development|ExecuteConsoleCommand "highresshot 3")
  (Development|PrintString "Face4")
  (Utilities|Time|SetTimerbyEvent (delegate Custom|CaptureFace5) 1.0))

(event Custom|CaptureFace5
  (Transformation|SetActorRotation (Variables|Default|GetCamera) "90.000000,0.000000,0.000000" false)
  (Development|ExecuteConsoleCommand "highresshot 3")
  (Development|PrintString "Face5")
  (Utilities|Time|SetTimerbyEvent (delegate Custom|CaptureFace6) 1.0))

(event Custom|CaptureFace6
  (Transformation|SetActorRotation (Variables|Default|GetCamera) "-90.000000,-0.000000,0.000000" false)
  (Development|ExecuteConsoleCommand "highresshot 3")
  (Development|PrintString "Face6")
  (Rendering|SetVisibility (Variables|Default|GetPlane6) (Variables|Default|GetShow%20Faces) false))

(event Custom|Capture%20Cubemap
  (Rendering|SetVisibility (Variables|Default|GetPlane6) true false)
  (CallFunction|ApplyTextures)
  (Rendering|SceneCapture|CaptureScene (Variables|Default|GetSceneCaptureComponentCube) :!class /Script/Engine.SceneCaptureComponentCube)
  (Rendering|SetVisibility (Variables|Default|GetPlane6) (Variables|Default|GetShow%20Faces) false))

(event Custom|Hide%20Faces
  (Rendering|SetVisibility (Variables|Default|GetPlane6) false false))

(event Custom|Apply%20Textures
  (bind _returnvalue (Rendering|Material|CreateDynamicMaterialInstance (Variables|Default|GetPlane) 0 "/Game/ArtTools/CubeMapCapture/M_Cubeface.M_Cubeface" "None" :!class /Script/Engine.PrimitiveComponent))
  (Rendering|Material|SetTextureParameterValue _returnvalue "Tex" (Variables|Default|GetFace1))
  (bind _returnvalue_2 (Rendering|Material|CreateDynamicMaterialInstance (Variables|Default|GetPlane2) 0 "/Game/ArtTools/CubeMapCapture/M_Cubeface.M_Cubeface" "None" :!class /Script/Engine.PrimitiveComponent))
  (Rendering|Material|SetTextureParameterValue _returnvalue_2 "Tex" (Variables|Default|GetFace2))
  (bind _returnvalue_3 (Rendering|Material|CreateDynamicMaterialInstance (Variables|Default|GetPlane3) 0 "/Game/ArtTools/CubeMapCapture/M_Cubeface.M_Cubeface" "None" :!class /Script/Engine.PrimitiveComponent))
  (Rendering|Material|SetTextureParameterValue _returnvalue_3 "Tex" (Variables|Default|GetFace3))
  (bind _returnvalue_4 (Rendering|Material|CreateDynamicMaterialInstance (Variables|Default|GetPlane4) 0 "/Game/ArtTools/CubeMapCapture/M_Cubeface.M_Cubeface" "None" :!class /Script/Engine.PrimitiveComponent))
  (Rendering|Material|SetTextureParameterValue _returnvalue_4 "Tex" (Variables|Default|GetFace4))
  (bind _returnvalue_5 (Rendering|Material|CreateDynamicMaterialInstance (Variables|Default|GetPlane5) 0 "/Game/ArtTools/CubeMapCapture/M_Cubeface.M_Cubeface" "None" :!class /Script/Engine.PrimitiveComponent))
  (Rendering|Material|SetTextureParameterValue _returnvalue_5 "Tex" (Variables|Default|GetFace5))
  (bind _returnvalue_6 (Rendering|Material|CreateDynamicMaterialInstance (Variables|Default|GetPlane6) 0 "/Game/ArtTools/CubeMapCapture/M_Cubeface.M_Cubeface" "None" :!class /Script/Engine.PrimitiveComponent))
  (Rendering|Material|SetTextureParameterValue _returnvalue_6 "Tex" (Variables|Default|GetFace6)))

(fn ConstructionScript
  (Rendering|SetVisibility (Variables|Default|GetPlane6) (Variables|Default|GetShow%20Faces) false))
