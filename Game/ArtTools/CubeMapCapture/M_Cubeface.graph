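# Unlit material that samples a single texture parameter, Tex, apparently as a cube face for the CubeMapCapture art tool.
# It derives a half-texel margin from the texture's size and clamps the UVs between that margin and one minus it.
# The inset keeps sampling on edge texel centers, so bilinear filtering does not wrap across face borders.
# Tex uses linear color sampling and defaults to a black 1x1 EXR placeholder.
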Material /Game/ArtTools/CubeMapCapture/M_Cubeface
graph {
 properties {shadingModel=MSM_Unlit, usedShadingModels=Unlit}
 nodes [
  n1 = TextureObjectParameter() {
   parameterName=Tex
   samplerType=SAMPLERTYPE_LinearColor
   texture=/Engine/Functions/Engine_MaterialFunctions02/PivotPainter2/Black_1x1_EXR_Texture
  }
  n2 = Divide(B=TextureProperty(Tex=@n1)) {constA=0.5}
  TextureSampleParameter2D(UVs=Clamp(None=TextureCoordinate(), Min=@n2, Max=OneMinus(None=@n2))) {
   parameterName=Tex
   samplerType=SAMPLERTYPE_LinearColor
   texture=&Black_1x1_EXR_Texture
  }
 ]
}
